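# Sound cue for a falling coconut, part of the Athena jungle foraged item sounds.
# It randomly picks one of two falling wave variations with equal weight, passes it through a modulator that lowers and slightly varies pitch, then adds a short fixed delay before playback.
# It uses close projectile attenuation and the ranged projectile effects sound class, with boosted volume.
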SoundCue /Game/Athena/Sounds/Items/JungleForaged/Athena_Coconut_Falling_Cue
root=Delay_0
nodes [
 {id=Delay_0, children [Modulator_0], properties {delayMax=0.2, delayMin=0.2}}
 {id=Modulator_0, children [Random_0], properties {pitchMax=0.81, pitchMin=0.76}}
 {id=Random_0, children [WavePlayer_1,WavePlayer_2], properties {weights [1,1]}}
 {id=WavePlayer_1, properties {soundWaveAssetPtr=&/Athena_Coconut_Falling_02}}
 {id=WavePlayer_2, properties {soundWaveAssetPtr=&/Athena_Coconut_Falling_01}}
]
properties {
 attenuationSettings=&Fort_Weapon_Projectile_Close_Athena_Att
 maxDistance=4950
 soundClassObject=&Fort_WeaponRanged_Projectile_Effects
 volumeMultiplier=1.3
}
