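# Sound cue for the impact of a coconut among the jungle foraged items.
# A random node picks one of two equally weighted impact waves, then a modulator applies slight pitch variation.
# It reuses the corn bounce attenuation and plays in the ranged projectile effects sound class.
# Overridden concurrency allows one instance per owner and blocks new ones while it plays.
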SoundCue /Game/Athena/Sounds/Items/JungleForaged/Athena_Coconut_Impact_Cue
root=Modulator_5
nodes [
 {id=Modulator_5, children [Random_1], properties {pitchMax=1.1, pitchMin=0.9}}
 {id=Random_1, children [WavePlayer_6,WavePlayer_7], properties {weights [1,1]}}
 {id=WavePlayer_6, properties {soundWaveAssetPtr=&/Athena_Coconut_Impact_02}}
 {id=WavePlayer_7, properties {soundWaveAssetPtr=&/Athena_Coconut_Impact_01}}
]
properties {
 attenuationSettings=/Game/Athena/Sounds/Items/Corn/Fort_Corn_Bounce_Att
 bOverrideConcurrency=true
 maxDistance=3000
 soundClassObject=&Fort_WeaponRanged_Projectile_Effects
 volumeMultiplier=1
 concurrencyOverrides {bLimitToOwner=true, maxCount=1, resolutionRule=PreventNew}
}
