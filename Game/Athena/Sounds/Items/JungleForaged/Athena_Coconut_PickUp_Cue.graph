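# Sound cue for picking up a foraged coconut in Athena.
# A random node chooses among four equally weighted coconut pickup waves, then a modulator varies pitch from normal to slightly higher.
# It plays at reduced volume using the corn bounce attenuation and a ranged-weapon projectile effects sound class.
# A concurrency override allows one instance per owner and blocks new plays while one is active.
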SoundCue /Game/Athena/Sounds/Items/JungleForaged/Athena_Coconut_PickUp_Cue
root=Modulator_5
nodes [
 {id=Modulator_5, children [Random_0], properties {pitchMax=1.2, pitchMin=1}}
 {id=Random_0, children [WavePlayer_4,WavePlayer_5,WavePlayer_6,WavePlayer_7], properties {weights [1,1,1,1]}}
 {id=WavePlayer_4, properties {soundWaveAssetPtr=&/Coconut_PickUp_04}}
 {id=WavePlayer_5, properties {soundWaveAssetPtr=&/Coconut_PickUp_03}}
 {id=WavePlayer_6, properties {soundWaveAssetPtr=&/Coconut_PickUp_02}}
 {id=WavePlayer_7, properties {soundWaveAssetPtr=&/Coconut_PickUp_01}}
]
properties {
 attenuationSettings=/Game/Athena/Sounds/Items/Corn/Fort_Corn_Bounce_Att
 bOverrideConcurrency=true
 maxDistance=3000
 soundClassObject=&Fort_WeaponRanged_Projectile_Effects
 volumeMultiplier=0.3
 concurrencyOverrides {bLimitToOwner=true, maxCount=1, resolutionRule=PreventNew}
}
